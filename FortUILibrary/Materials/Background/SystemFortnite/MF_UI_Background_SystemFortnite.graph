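# Material function that builds the animated Fortnite system UI background and outputs it as Color.
# Three palette colors come from MPC_Background_Fortnite or local parameters via the Use Unified Color static switch.
# Named reroutes carry layers for a radial gradient and glow, time-panned noise clouds, an animated wave, triangle highlights and a vignette.
# Screen-space UVs drive the layout, with slowly rotating triangle UVs offset by the collection's ScrollingAmount.
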MaterialFunction /FortUILibrary/Materials/Background/SystemFortnite/MF_UI_Background_SystemFortnite
graph {nodes [
  Constant() {r=0.5}
  n1 = NamedRerouteUsage()
  n2 = ComponentMask(None=NamedRerouteUsage()) {r=true}
  n3 = Min(A=@n2, B=OneMinus(None=@n2))
  n4 = OneMinus(None=Saturate(None=Multiply(A=@n3) {constB=2}))
  n5 = LinearInterpolate(A=Multiply(A=NamedRerouteUsage(), B=@n1), B=@n1, Alpha=Multiply(A=Power(Base=@n4) {constExponent=3}) {constB=0.15})
  FunctionOutput(None=ComponentMask(None=Saturate(None=@n5)) {b=true, g=true, r=true}) {outputName=Color}
  n6 = NamedRerouteUsage()
  n7 = Multiply(A=MaterialFunctionCall(Input=ComponentMask(None=@n6) {g=true}, Input Low=Constant() {r=0.275}, Input High=Constant() {r=1}) {materialFunction=&RemapValue_0-1}.Result) {constB=0.9}
  n8 = NamedRerouteUsage()
  n9 = Length(None=Multiply(A=Subtract(A=@n8) {constB=0.5}) {constB=0.2667}) {is=class MaterialExpressionLength}
  n10 = Saturate(None=SmoothStep(Value=Saturate(None=@n9)) {constMax=0.5})
  n11 = NamedRerouteDeclaration(None=LinearInterpolate(A=NamedRerouteUsage(), B=NamedRerouteUsage(), Alpha=LinearInterpolate(A=@n7, Alpha=@n10) {constB=0.95})) {name=Vignette, nodeColor=<rgba 0 0 0 1>}
  n12 = Multiply(A=NamedRerouteUsage(), B=NamedRerouteUsage())
  LinearInterpolate(A=@n11, B=Multiply(A=@n11) {constB=0.5}, Alpha=Multiply(A=@n12, B=@n12))
  n13 = NamedRerouteDeclaration(None=ScreenPosition().ViewportUV) {name=N UV, nodeColor=<rgba 0 0 0 1>}
  MaterialFunctionCall(UV=@n13) {materialFunction=&MF_UI_UVScaling}
  MaterialFunctionCall() {materialFunction=&GetUserInterfaceUV}
  n14 = CollectionParameter() {collection=&/MPC_Background_Fortnite, parameterName=PrimaryColor}
  n15 = VectorParameter() {group=Background Visuals, parameterName=PrimaryColor, sortPriority=0, defaultValue=<rgba 0 0.3231 0.9823 1>}
  NamedRerouteDeclaration(None=ComponentMask(None=StaticSwitchParameter(True=@n14, False=@n15.RGB) {defaultValue=true, group=Background Visuals, parameterName=Use Unified Color, sortPriority=-1}) {b=true, g=true, r=true}) {name=C1, nodeColor=<rgba 0 0 0 1>}
  n16 = CollectionParameter() {collection=&MPC_Background_Fortnite, parameterName=SecondaryColor}
  n17 = VectorParameter() {group=Background Visuals, parameterName=SecondaryColor, sortPriority=5, defaultValue=<rgba 0 0.1441 0.5972 1>}
  NamedRerouteDeclaration(None=ComponentMask(None=StaticSwitchParameter(True=@n16, False=@n17.RGB) {defaultValue=true, group=Background Visuals, parameterName=Use Unified Color, sortPriority=-1}) {b=true, g=true, r=true}) {name=C2, nodeColor=<rgba 0 0 0 1>}
  n18 = NamedRerouteUsage()
  n19 = Rotator(Coordinate=Add(A=@n18, B=Constant2Vector() {g=-0.5, r=0.5}), Time=Constant() {r=0.425}) {speed=6.2832}
  n20 = Add(A=Multiply(A=Subtract(A=@n19) {constB=0.5}, B=Constant2Vector() {g=0.6061, r=0.404}) {constB=0.2667}) {constB=0.5}
  n21 = Constant() {r=1}
  n22 = Constant() {r=3}
  n23 = Constant2Vector() {g=0.15}
  n24 = ComponentMask(None=Rotator(Coordinate=Add(A=@n18, B=@n23), Time=Constant() {r=0.18}) {speed=6.2832}) {r=true}
  n25 = Multiply(A=Min(A=@n24, B=OneMinus(None=@n24))) {constB=2}
  NamedRerouteDeclaration(None=Multiply(
    A=MaterialFunctionCall!(UVs=@n20, Size=@n21, Sides=@n22, InnerAngle, OutlineThickness, OutlineAlignment, GlowMin, GlowMax) {materialFunction=/FortUILibrary/MaterialFunctions/Shapes/MF_UI_RegularPolygon}.Fill
    B=Saturate(None=@n25)
   )) {name=Left Triangle Mask, nodeColor=<rgba 0 0 0 1>}
  n26 = NamedRerouteUsage()
  n27 = Rotator(Coordinate=Add(A=@n26, B=Constant2Vector() {r=-0.85}), Time=Constant() {r=0.535}) {speed=6.2832}
  n28 = Add(A=Multiply(A=Subtract(A=@n27) {constB=0.5}, B=Constant2Vector() {g=0.8081, r=0.4848}) {constB=0.2667}) {constB=0.5}
  n29 = Constant() {r=1}
  n30 = Constant() {r=3}
  n31 = Constant2Vector() {g=0.6}
  n32 = ComponentMask(None=Rotator(Coordinate=Add(A=@n26, B=@n31), Time=Constant() {r=0.2}) {speed=6.2832}) {r=true}
  NamedRerouteDeclaration(None=Multiply(
    A=MaterialFunctionCall(UVs=@n28, Size=@n29, Sides=@n30) {materialFunction=&MF_UI_RegularPolygon}.Fill
    B=Saturate(None=@n32)
   )) {name=Right Triangle Mask, nodeColor=<rgba 0 0 0 1>}
  n33 = MaterialFunctionCall!(ScreenUV=@n13, TextureSize, ScalePivotPoint) {materialFunction=/FortUILibrary/MaterialFunctions/Basics/MF_UI_ScreenSpaceUVScaling}
  NamedRerouteDeclaration(None=@n33.ScaleX) {name=ScaleX, nodeColor=<rgba 0 0 0 1>}
  n34 = CollectionParameter() {collection=&MPC_Background_Fortnite, parameterName=TertiaryColor}
  n35 = VectorParameter() {group=Background Visuals, parameterName=TertiaryColor, sortPriority=10, defaultValue=<rgba 0.008 0.008 0.016 1>}
  NamedRerouteDeclaration(None=ComponentMask(None=StaticSwitchParameter(True=@n34, False=@n35.RGB) {defaultValue=true, group=Background Visuals, parameterName=Use Unified Color, sortPriority=-1}) {b=true, g=true, r=true}) {name=C3, nodeColor=<rgba 0 0 0 1>}
  n36 = Constant()
  n37 = CollectionParameter() {collection=&MPC_Background_Fortnite, parameterName=ScrollingAmount}
  n38 = Rotator(Coordinate=Add(A=NamedRerouteUsage(), B=AppendVector(A=@n36, B=@n37))) {speed=0.02}
  n39 = Add(A=Multiply(A=Subtract(A=@n38) {constB=0.5}, B=Constant2Vector() {g=0.2222, r=0.5333}) {constB=0.2667}) {constB=0.5}
  n40 = Constant2Vector() {g=3, r=1}
  NamedRerouteDeclaration(None=Frac(None=Multiply(A=@n39, B=@n40))) {name=Scroll Triangles UV, nodeColor=<rgba 0 0 0 1>}
  NamedRerouteDeclaration(None=@n33.ScaleY) {name=ScaleY, nodeColor=<rgba 0 0 0 1>}
  n41 = Length(None=Subtract(A=NamedRerouteUsage()) {constB=0.5})
  n42 = SmoothStep(Value=Saturate(None=OneMinus(None=@n41)))
  NamedRerouteDeclaration(None=LinearInterpolate(A=NamedRerouteUsage(), B=NamedRerouteUsage(), Alpha=NamedRerouteDeclaration(None=@n42) {name=Radial Mask, nodeColor=<rgba 0 0 0 1>})) {name=Background Gradient, nodeColor=<rgba 0 0 0 1>}
  n43 = Multiply(A=Subtract(A=NamedRerouteUsage()) {constB=0.5}) {constB=0.25}
  n44 = Time()
  n45 = TextureSample(UVs=Add(A=Add(A=@n43) {constB=0.5}, B=Multiply(A=@n44) {constB=0.01})) {samplerSource=SSM_Wrap_WorldGroupSettings, samplerType=SAMPLERTYPE_Grayscale, texture=&LowResBlurredNoise}
  n46 = Multiply(A=Time()) {constB=0.01}
  n47 = ConstantBiasScale(None=Sine(None=Subtract(A=@n45.R, B=@n46)))
  NamedRerouteDeclaration(None=Multiply(A=@n47, B=@n47)) {name=Cloud Noise, nodeColor=<rgba 0 0 0 1>}
  n48 = LinearInterpolate(A=NamedRerouteUsage(), Alpha=NamedRerouteUsage()) {constB=0.9}
  NamedRerouteDeclaration(None=LinearInterpolate(A=NamedRerouteUsage(), B=NamedRerouteUsage(), Alpha=Multiply(A=@n48) {constB=0.4}) {constAlpha=0.4}) {name=Radial Glow, nodeColor=<rgba 0 0 0 1>}
  n49 = NamedRerouteUsage()
  n50 = Saturate(None=Add(A=NamedRerouteUsage(), B=LinearInterpolate(Alpha=@n49) {constB=0.025}))
  n51 = NamedRerouteUsage()
  n52 = NamedRerouteUsage()
  n53 = Constant2Vector() {g=0.375}
  n54 = OneMinus(None=MaterialFunctionCall!(
    UV=Add(A=@n52, B=@n53)
    Wave Height=Constant() {r=0.25}
    Wave Length=Constant() {r=0.75}
    Wave Speed=Constant() {r=0.1}
    GlowMin
    GlowMax
    Outline Thickness=Constant() {r=0.1}
    Custom Curve
   ) {materialFunction=/FortUILibrary/MaterialFunctions/Shapes/MF_UI_Wave}.SDF Outline)
  n55 = NamedRerouteDeclaration(None=Saturate(None=SmoothStep(Value=@n54) {constMin=-0.25})) {name=Wave, nodeColor=<rgba 0 0 0 1>}
  NamedRerouteDeclaration(None=Saturate(None=LinearInterpolate(A=@n50, B=@n51, Alpha=@n55))) {name=Clouds Gradient, nodeColor=<rgba 0 0 0 1>}
  n56 = NamedRerouteUsage()
  n57 = Normalize(VectorInput=Max(A=@n56) {constB=0})
  n58 = NamedRerouteUsage()
  n59 = NamedRerouteUsage()
  n60 = Multiply(A=Saturate(None=Add(A=@n58, B=@n59))) {constB=2}
  NamedRerouteDeclaration(None=LinearInterpolate(A=@n56, B=Saturate(None=@n57), Alpha=Saturate(None=@n60))) {name=Triangles, nodeColor=<rgba 0 0 0 1>}
 ]}
